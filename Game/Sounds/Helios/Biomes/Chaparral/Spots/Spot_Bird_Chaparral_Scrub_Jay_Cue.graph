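# Sound cue for a one-shot scrub jay bird call, an outdoor ambient spot sound for the Chaparral biome.
# A quality-level switch sends its first two slots to seven equally weighted random call variations and its last two to a smaller two-wave set.
# Each branch passes through a modulator at default settings.
# A bird one-shot attenuation asset spatializes it, with a distance-based send to the exterior ambient source bus.
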SoundCue /Game/Sounds/Helios/Biomes/Chaparral/Spots/Spot_Bird_Chaparral_Scrub_Jay_Cue
root=QualityLevel_0
nodes [
 {id=Modulator_0, children [Random_1]}
 {id=Modulator_1, children [Random_2]}
 {id=QualityLevel_0, children [Modulator_0,Modulator_0,Modulator_1,Modulator_1]}
 {
  id=Random_1
  children [WavePlayer_9, WavePlayer_10, WavePlayer_11, WavePlayer_12, WavePlayer_13, WavePlayer_14, WavePlayer_15]
  properties {weights [1,1,1,1,1,1,1]}
 }
 {id=Random_2, children [WavePlayer_18, WavePlayer_19], properties {weights [1,1]}}
 {id=WavePlayer_10, properties {soundWaveAssetPtr=&/Spot_Bird_Scrub_Jay_02}}
 {id=WavePlayer_11, properties {soundWaveAssetPtr=&/Spot_Bird_Scrub_Jay_03}}
 {id=WavePlayer_12, properties {soundWaveAssetPtr=&/Spot_Bird_Scrub_Jay_04}}
 {id=WavePlayer_13, properties {soundWaveAssetPtr=&/Spot_Bird_Scrub_Jay_05}}
 {id=WavePlayer_14, properties {soundWaveAssetPtr=&/Spot_Bird_Scrub_Jay_06}}
 {id=WavePlayer_15, properties {soundWaveAssetPtr=&/Spot_Bird_Scrub_Jay_07}}
 {id=WavePlayer_18, properties {soundWaveAssetPtr=&/Spot_Bird_Scrub_Jay_01}}
 {id=WavePlayer_19, properties {soundWaveAssetPtr=&Spot_Bird_Scrub_Jay_03}}
 {id=WavePlayer_9, properties {soundWaveAssetPtr=&Spot_Bird_Scrub_Jay_01}}
]
properties {
 attenuationSettings=&Fort_TileAmbient_OneShot_Bird_Att
 maxDistance=8005
 soundClassObject=&Fort_Ambient_OneShots_Outdoors
 volumeMultiplier=0.276
 busSends [{
   audioBus=null
   maxSendDistance=6000
   minSendDistance=3000
   minSendLevel=0.6
   sendLevel=1
   soundSourceBus=&AmbientAudio_Exterior_SourceBus
  }]
}
