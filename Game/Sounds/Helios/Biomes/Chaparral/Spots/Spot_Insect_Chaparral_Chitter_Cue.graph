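# Sound cue for insect chitter spot one-shots in the outdoor ambience of the Helios Chaparral biome.
# A quality-level node sends its first two levels to a random pick of twelve variants and its last two to three variants, each pool passing through a modulator.
# It plays at low volume with shared bird one-shot attenuation and a distance-ranged send to the exterior ambient source bus.
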SoundCue /Game/Sounds/Helios/Biomes/Chaparral/Spots/Spot_Insect_Chaparral_Chitter_Cue
root=QualityLevel_0
nodes [
 {id=Modulator_0, children [Random_2]}
 {id=Modulator_1, children [Random_3]}
 {id=QualityLevel_0, children [Modulator_0,Modulator_0,Modulator_1,Modulator_1]}
 {
  id=Random_2
  children [
   WavePlayer_27
   WavePlayer_28
   WavePlayer_29
   WavePlayer_30
   WavePlayer_31
   WavePlayer_32
   WavePlayer_33
   WavePlayer_34
   WavePlayer_35
   WavePlayer_36
   WavePlayer_37
   WavePlayer_38
  ]
  properties {weights [1,1,1,1,1,1,1,1,1,1,1,1]}
 }
 {id=Random_3, children [WavePlayer_39, WavePlayer_40, WavePlayer_41], properties {weights [1,1,1]}}
 {id=WavePlayer_27, properties {soundWaveAssetPtr=&/Spot_Chaparral_Insect_Chitter_01}}
 {id=WavePlayer_28, properties {soundWaveAssetPtr=&/Spot_Chaparral_Insect_Chitter_02}}
 {id=WavePlayer_29, properties {soundWaveAssetPtr=&/Spot_Chaparral_Insect_Chitter_03}}
 {id=WavePlayer_30, properties {soundWaveAssetPtr=&/Spot_Chaparral_Insect_Chitter_04}}
 {id=WavePlayer_31, properties {soundWaveAssetPtr=&/Spot_Chaparral_Insect_Chitter_05}}
 {id=WavePlayer_32, properties {soundWaveAssetPtr=&/Spot_Chaparral_Insect_Chitter_06}}
 {id=WavePlayer_33, properties {soundWaveAssetPtr=&/Spot_Chaparral_Insect_Chitter_07}}
 {id=WavePlayer_34, properties {soundWaveAssetPtr=&/Spot_Chaparral_Insect_Chitter_08}}
 {id=WavePlayer_35, properties {soundWaveAssetPtr=&/Spot_Chaparral_Insect_Chitter_09}}
 {id=WavePlayer_36, properties {soundWaveAssetPtr=&/Spot_Chaparral_Insect_Chitter_10}}
 {id=WavePlayer_37, properties {soundWaveAssetPtr=&/Spot_Chaparral_Insect_Chitter_11}}
 {id=WavePlayer_38, properties {soundWaveAssetPtr=&/Spot_Chaparral_Insect_Chitter_12}}
 {id=WavePlayer_39, properties {soundWaveAssetPtr=&Spot_Chaparral_Insect_Chitter_03}}
 {id=WavePlayer_40, properties {soundWaveAssetPtr=&Spot_Chaparral_Insect_Chitter_02}}
 {id=WavePlayer_41, properties {soundWaveAssetPtr=&Spot_Chaparral_Insect_Chitter_01}}
]
properties {
 attenuationSettings=&Fort_TileAmbient_OneShot_Bird_Att
 maxDistance=8005
 soundClassObject=&Fort_Ambient_OneShots_Outdoors
 volumeMultiplier=0.125
 busSends [{
   audioBus=null
   maxSendDistance=6000
   minSendDistance=3000
   minSendLevel=0.6
   sendLevel=1
   soundSourceBus=&AmbientAudio_Exterior_SourceBus
  }]
}
