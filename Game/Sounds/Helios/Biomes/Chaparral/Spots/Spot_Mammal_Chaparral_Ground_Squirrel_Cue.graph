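# Sound cue for a one-shot ground squirrel spot ambience in the Chaparral biome.
# An equal-weight random pick of three squirrel waves feeds a volume modulator for slight level variation, shared by all four quality-level inputs.
# It uses the outdoor ambient one-shot sound class, borrows a bird one-shot attenuation, and sends to the exterior ambient source bus over a distance range.
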SoundCue /Game/Sounds/Helios/Biomes/Chaparral/Spots/Spot_Mammal_Chaparral_Ground_Squirrel_Cue
root=QualityLevel_0
nodes [
 {id=Modulator_0, children [Random_0], properties {volumeMax=0.6, volumeMin=0.5}}
 {id=QualityLevel_0, children [Modulator_0,Modulator_0,Modulator_0,Modulator_0]}
 {id=Random_0, children [WavePlayer_3,WavePlayer_4,WavePlayer_5], properties {weights [1,1,1]}}
 {id=WavePlayer_3, properties {soundWaveAssetPtr=&/Spot_Ground_Squirrel_01}}
 {id=WavePlayer_4, properties {soundWaveAssetPtr=&/Spot_Ground_Squirrel_02}}
 {id=WavePlayer_5, properties {soundWaveAssetPtr=&/Spot_Ground_Squirrel_03}}
]
properties {
 attenuationSettings=&Fort_TileAmbient_OneShot_Bird_Att
 maxDistance=8005
 soundClassObject=&Fort_Ambient_OneShots_Outdoors
 volumeMultiplier=0.276
 busSends [{
   audioBus=null
   maxSendDistance=6000
   minSendDistance=3000
   minSendLevel=0.6
   sendLevel=1
   soundSourceBus=&AmbientAudio_Exterior_SourceBus
  }]
}
