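# Sound cue for one-shot outdoor ambient wind gust spots in the Chaparral biome.
# A quality-level root sends the first two levels to a random pick of eighteen gusts and the last two to a random pick of four.
# Each pool passes through its own modulator for pitch and volume variation.
# The cue plays quieter and lower-pitched overall, using wind one-shot attenuation and the outdoor ambient sound class.
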SoundCue /Game/Sounds/Helios/Biomes/Chaparral/Spots/Spot_Wind_Gusts_Chaparral_Cue
root=QualityLevel_1
nodes [
 {id=Modulator_1, children [Random_6], properties {pitchMax=1.1, pitchMin=0.9, volumeMax=1, volumeMin=0.6}}
 {id=Modulator_2, children [Random_1], properties {pitchMax=1.1, volumeMax=1, volumeMin=0.7}}
 {id=QualityLevel_1, children [Modulator_1,Modulator_1,Modulator_2,Modulator_2]}
 {id=Random_1, children [WavePlayer_44, WavePlayer_45, WavePlayer_46, WavePlayer_47], properties {weights [1,1,1,1]}}
 {
  id=Random_6
  children [
   WavePlayer_20
   WavePlayer_27
   WavePlayer_28
   WavePlayer_29
   WavePlayer_30
   WavePlayer_31
   WavePlayer_32
   WavePlayer_33
   WavePlayer_34
   WavePlayer_35
   WavePlayer_36
   WavePlayer_37
   WavePlayer_38
   WavePlayer_39
   WavePlayer_40
   WavePlayer_41
   WavePlayer_42
   WavePlayer_43
  ]
  properties {weights [1,1,1,1,1,1,1,1,1,1,1,1,1,1,1,1,1,1]}
 }
 {id=WavePlayer_20, properties {soundWaveAssetPtr=&/Spot_Chaparral_Wind_Gust_18}}
 {id=WavePlayer_27, properties {soundWaveAssetPtr=&/Spot_Chaparral_Wind_Gust_17}}
 {id=WavePlayer_28, properties {soundWaveAssetPtr=&/Spot_Chaparral_Wind_Gust_16}}
 {id=WavePlayer_29, properties {soundWaveAssetPtr=&/Spot_Chaparral_Wind_Gust_15}}
 {id=WavePlayer_30, properties {soundWaveAssetPtr=&/Spot_Chaparral_Wind_Gust_14}}
 {id=WavePlayer_31, properties {soundWaveAssetPtr=&/Spot_Chaparral_Wind_Gust_13}}
 {id=WavePlayer_32, properties {soundWaveAssetPtr=&/Spot_Chaparral_Wind_Gust_12}}
 {id=WavePlayer_33, properties {soundWaveAssetPtr=&/Spot_Chaparral_Wind_Gust_11}}
 {id=WavePlayer_34, properties {soundWaveAssetPtr=&/Spot_Chaparral_Wind_Gust_10}}
 {id=WavePlayer_35, properties {soundWaveAssetPtr=&/Spot_Chaparral_Wind_Gust_09}}
 {id=WavePlayer_36, properties {soundWaveAssetPtr=&/Spot_Chaparral_Wind_Gust_08}}
 {id=WavePlayer_37, properties {soundWaveAssetPtr=&/Spot_Chaparral_Wind_Gust_07}}
 {id=WavePlayer_38, properties {soundWaveAssetPtr=&/Spot_Chaparral_Wind_Gust_06}}
 {id=WavePlayer_39, properties {soundWaveAssetPtr=&/Spot_Chaparral_Wind_Gust_05}}
 {id=WavePlayer_40, properties {soundWaveAssetPtr=&/Spot_Chaparral_Wind_Gust_04}}
 {id=WavePlayer_41, properties {soundWaveAssetPtr=&/Spot_Chaparral_Wind_Gust_03}}
 {id=WavePlayer_42, properties {soundWaveAssetPtr=&/Spot_Chaparral_Wind_Gust_02}}
 {id=WavePlayer_43, properties {soundWaveAssetPtr=&/Spot_Chaparral_Wind_Gust_01}}
 {id=WavePlayer_44, properties {soundWaveAssetPtr=&Spot_Chaparral_Wind_Gust_04}}
 {id=WavePlayer_45, properties {soundWaveAssetPtr=&Spot_Chaparral_Wind_Gust_03}}
 {id=WavePlayer_46, properties {soundWaveAssetPtr=&Spot_Chaparral_Wind_Gust_02}}
 {id=WavePlayer_47, properties {soundWaveAssetPtr=&Spot_Chaparral_Wind_Gust_01}}
]
properties {
 attenuationSettings=/Game/Sounds/Attenuation/Fort_TileAmbient_OneShot_Wind_Att
 bEnableBusSends=false
 maxDistance=8000
 pitchMultiplier=0.8
 soundClassObject=&Fort_Ambient_OneShots_Outdoors
 volumeMultiplier=0.125
 preEffectBusSends [{}]
}
